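# List entry widget for the active player augments panel, showing an augment's icon and text or an empty-slot prompt.
# SetAugmentInfo updates the inherited entry and async-loads the icon into a dynamic material.
# Parent-class events toggle "Unlocked" and borrowed-from-teammate callouts, tinting the lender's avatar backing by pin color.
# Material-driven animations handle spawn and focus (reversed on unfocus), and construction adjusts alignment and height for mobile touch.
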WidgetBlueprint /PlayerAugmentsUI/UI/ActiveAugmentsPanel/WBP_ActiveAugments_ListEntry
animations [
 Anim_SpawnEntry {
  binding_count=3
  length_seconds=1.5
  object=&self:Anim_SpawnEntry
  sequence {
   bindings [
    Icon {
     id=#1
     possessed_class=Overlay
     tracks [
      {
       type=MovieSceneFloatTrack
       display_name=Render Opacity
       sections [{type=MovieSceneFloatSection, end_frame=90, start_frame=0, channels [None {type=MovieSceneScriptingFloatChannel, keys [[30,0],[90,1]]}]}]
      }
      {
       type=MovieScene2DTransformTrack
       sections [{type=MovieScene2DTransformSection, channels [Translation.Y {type=MovieSceneScriptingFloatChannel, keys [[30,64],[60,0]]}]}]
      }
     ]
    }
    Backing_Panel {
     id=#2
     possessed_class=Image
     tracks [{
       type=MovieSceneWidgetMaterialTrack
       sections [{
         type=MovieSceneParameterSection
         channels [
          Anim_ScanLines_Amount {type=MovieSceneScriptingFloatChannel, keys [[0,0],[9,0],[18,16],[69,0]]}
          ScanLines_Speed {type=MovieSceneScriptingFloatChannel, keys [[0,0.333],[9,0.333],[18,2],[30,0.333]]}
         ]
        }]
      }]
    }
    VisSwitcher_Description {
     id=#3
     possessed_class=CommonVisibilitySwitcher
     tracks [
      {
       type=MovieSceneFloatTrack
       display_name=Render Opacity
       sections [{type=MovieSceneFloatSection, channels [None {type=MovieSceneScriptingFloatChannel, keys [[39,0],[90,1]]}]}]
      }
      {
       type=MovieScene2DTransformTrack
       sections [{
         type=MovieScene2DTransformSection
         channels [
          Translation.X {type=MovieSceneScriptingFloatChannel, keys [[39,16],[69,0]]}
          Translation.Y {type=MovieSceneScriptingFloatChannel, keys [[39,16],[69,0]]}
          Scale.X {type=MovieSceneScriptingFloatChannel, keys [[39,0.9],[69,1]]}
          Scale.Y {type=MovieSceneScriptingFloatChannel, keys [[39,0.9],[69,1]]}
         ]
        }]
      }
     ]
    }
   ]
   display_rate {numerator=60}
  }
 }
 Anim_Focus {
  binding_count=12
  length_seconds=0.25
  object=&self:Anim_Focus
  sequence {
   bindings [
    Text_AugmentDescription {
     id=#4
     possessed_class=CommonTextBlock
     tracks [{
       type=MovieSceneWidgetMaterialTrack
       display_name=Font.FontMaterial
       sections [{
         type=MovieSceneParameterSection
         channels [
          Anim_BackingAlpha {type=MovieSceneScriptingFloatChannel, keys [[0,0],[15,1]]}
          Anim_Color {type=MovieSceneScriptingFloatChannel, keys [[0,0],[15,1]]}
         ]
        }]
      }]
    }
    TextPanelBrackets {
     id=#5
     possessed_class=Image
     tracks [{
       type=MovieSceneWidgetMaterialTrack
       sections [{
         type=MovieSceneParameterSection
         channels [
          Border_Scale_X {type=MovieSceneScriptingFloatChannel, const [1.1,2]}
          Anim_BorderAlpha {type=MovieSceneScriptingFloatChannel, keys [[0,0],[15,1]]}
         ]
        }]
      }]
    }
    Text_AugmentName {
     id=#6
     possessed_class=CommonTextBlock
     tracks [{
       type=MovieSceneWidgetMaterialTrack
       display_name=Font.FontMaterial
       sections [{
         type=MovieSceneParameterSection
         channels [
          Anim_BackingAlpha {type=MovieSceneScriptingFloatChannel, keys [[0,0],[15,1]]}
          Anim_Color {type=MovieSceneScriptingFloatChannel, keys [[0,0],[15,1]]}
         ]
        }]
      }]
    }
    Image_AugmentIcon {
     id=#7
     possessed_class=Image
     tracks [{
       type=MovieSceneWidgetMaterialTrack
       sections [{
         type=MovieSceneParameterSection
         channels [
          Anim_FinalAlpha {type=MovieSceneScriptingFloatChannel, keys [[0,0],[15,1]]}
          Anim_IconColor {type=MovieSceneScriptingFloatChannel, keys [[0,0],[15,1]]}
          Anim_BackingAlpha {type=MovieSceneScriptingFloatChannel, keys [[0,0],[15,1]]}
          Anim_GlowAmount {type=MovieSceneScriptingFloatChannel, keys [[0,0],[15,1]]}
         ]
        }]
      }]
    }
    Icon {
     id=#8
     possessed_class=Overlay
     tracks [{
       type=MovieScene2DTransformTrack
       sections [{
         type=MovieScene2DTransformSection
         channels [
          Translation.X {type=MovieSceneScriptingFloatChannel, keys [[0,0],[15,-4]]}
          Translation.Y {type=MovieSceneScriptingFloatChannel, keys [[0,0],[15,-4]]}
          Scale.X {type=MovieSceneScriptingFloatChannel, keys [[0,1],[15,1.05]]}
          Scale.Y {type=MovieSceneScriptingFloatChannel, keys [[0,1],[15,1.05]]}
         ]
        }]
      }]
    }
    Icon_Border {
     id=#9
     possessed_class=Image
     tracks [{
       type=MovieSceneWidgetMaterialTrack
       sections [{type=MovieSceneParameterSection, channels [Anim_BorderAlpha {type=MovieSceneScriptingFloatChannel, keys [[0,0],[15,1]]}]}]
      }]
    }
    Image_SolidBG_AugmentDescription {
     id=#10
     possessed_class=Image
     tracks [{
       type=MovieSceneWidgetMaterialTrack
       sections [{
         type=MovieSceneParameterSection
         channels [
          Anim_BackingAlpha {type=MovieSceneScriptingFloatChannel, keys [[0,0],[15,1]]}
          Anim_Color {type=MovieSceneScriptingFloatChannel, keys [[0,0],[15,1]]}
         ]
        }]
      }]
    }
    Image_SolidBG_AugmentDescription_Reflection {
     id=#11
     possessed_class=Image
     tracks [{
       type=MovieSceneWidgetMaterialTrack
       sections [{type=MovieSceneParameterSection, start_frame=0, channels [Anim_BackingAlpha {type=MovieSceneScriptingFloatChannel, keys [[0,0],[15,1]]}]}]
      }]
    }
    Text_AugmentCategory {
     id=#12
     possessed_class=CommonTextBlock
     tracks [{
       type=MovieSceneWidgetMaterialTrack
       display_name=Font.FontMaterial
       sections [{
         type=MovieSceneParameterSection
         channels [
          Anim_BackingAlpha {type=MovieSceneScriptingFloatChannel, keys [[0,0],[15,1]]}
          Anim_Color {type=MovieSceneScriptingFloatChannel, keys [[0,0],[15,1]]}
         ]
        }]
      }]
    }
    Icon_BackingPanel {
     id=#13
     possessed_class=Image
     tracks [{
       type=MovieSceneWidgetMaterialTrack
       sections [{
         type=MovieSceneParameterSection
         channels [
          Anim_BackingAlpha {type=MovieSceneScriptingFloatChannel, keys [[0,0],[15,1]]}
          Anim_Color {type=MovieSceneScriptingFloatChannel, keys [[0,0],[15,1]]}
         ]
        }]
      }]
    }
    Backing_EdgeGlow {
     id=#14
     possessed_class=Image
     tracks [{
       type=MovieSceneWidgetMaterialTrack
       sections [{type=MovieSceneParameterSection, channels [Anim_BorderAlpha {type=MovieSceneScriptingFloatChannel, keys [[0,0],[15,1]]}]}]
      }]
    }
    Backing_Panel {
     id=#15
     possessed_class=Image
     tracks [{
       type=MovieSceneWidgetMaterialTrack
       sections [{
         type=MovieSceneParameterSection
         channels [
          Anim_BackingAlpha {type=MovieSceneScriptingFloatChannel, keys [[0,0],[15,1]]}
          Anim_ScanLines_Amount {type=MovieSceneScriptingFloatChannel, keys [[0,0],[6,16],[15,0]]}
          Anim_Color {type=MovieSceneScriptingFloatChannel, keys [[0,0],[15,1]]}
         ]
        }]
      }]
    }
   ]
   display_rate {numerator=60}
  }
 }
]
graphs [
 EventGraph {nodes [
   n1 = SetAugmentIcon!() -> (OutputDelegate, then, Asset:Soft Object Reference) {is=title K2Node_CustomEvent}
   > n2 = Async Load Asset(Asset=@n1.Asset)
   n3 = Cast To Texture2D(execute=@n2.Completed, Object=@n2)
   > n4 = GetDynamicMaterial(self=Get Image_AugmentIcon() -> (Image_AugmentIcon:Image Object Reference))
   > SetTextureParameterValue(self=@n4, ParameterName=Icon, Value=@n3)
   n5 = Event OnItemNewOrBorrowedSet!() -> (OutputDelegate, then, bIsNew:bool, bIsBorrowed:bool) {is=title K2Node_Event}
   > UpdateNewPrompt!(execute, self, IsNew:bool=@n5.bIsNew) -> (then)
   > UpdateBorrowedPrompt!(execute, self, IsBorrowed:bool=@n5.bIsBorrowed) -> (then)
   Event Construct()
   > SetMobileVisuals()
   n6 = Event OnItemPinColorSet!() -> (OutputDelegate, then, PinColor:Linear Color Structure) {is=title K2Node_Event}
   > n7 = Sequence()
   > n8 = GetDynamicMaterial(self=Get Image_TeammateIcon_Color() -> (Image_TeammateIcon_Color:Image Object Reference))
   > SetVectorParameterValue(self=@n8, ParameterName=Backing_Color, Value=@n6.PinColor)
   n9 = GetDynamicMaterial(execute=@n7.then_1, self=Get Image_TeammateIcon_Color_Glow() -> (Image_TeammateIcon_Color_Glow:Image Object Reference))
   > SetVectorParameterValue(self=@n9, ParameterName=Backing_Color, Value=@n6.PinColor)
  ]}
 PlaySpawnEntryAnim {nodes [
   PlaySpawnEntryAnim() {is=title K2Node_FunctionEntry}
   > PlayAnimation(InAnimation=Get Anim_SpawnEntry() -> (Anim_SpawnEntry:Widget Animation Object Reference), PlaybackSpeed=1.75)
  ]}
 SetAsEmptySlot {nodes [
   n1 = SetAsEmptySlot!() -> (then, InIsEmptySlot:bool=true) {is=title K2Node_FunctionEntry}
   > Branch(Condition="Not Equal (Boolean)"(A=@n1, B=Get IsEmptySlot() -> (IsEmptySlot:bool)) -> (ReturnValue=~))
   > Set IsEmptySlot(IsEmptySlot:bool=@n1) -> (Output_Get:bool)
   > SetActiveWidget(
    self=Get VisSwitcher_Description() -> (VisSwitcher_Description:Common Visibility Switcher Object Reference)
    Widget=Select(
     Option 0:Widget Object Reference=Get VBox_AugmentText() -> (VBox_AugmentText:Vertical Box Object Reference)
     Option 1:Widget Object Reference=Get Overlay_EmptyPrompt() -> (Overlay_EmptyPrompt:Overlay Object Reference)
     Index=Get IsEmptySlot()
    ) -> (ReturnValue:Widget Object Reference)
   )
   > SetVisibility(
    self=[
     Get Icon_Border_Unlocked() -> (Icon_Border_Unlocked:Image Object Reference)
     Get Icon_Border_UnlockedGlow() -> (Icon_Border_UnlockedGlow:Image Object Reference)
     Get Image_AugmentIcon()
     Get Overlay_NewPrompt() -> (Overlay_NewPrompt:Overlay Object Reference)
     Get Icon_BackingPanel_UnlockedFX() -> (Icon_BackingPanel_UnlockedFX:Image Object Reference)
    ]
    InVisibility=Select(Option 0:ESlateVisibility Enum=SelfHitTestInvisible, Option 1:ESlateVisibility Enum=Collapsed, Index=Get IsEmptySlot()) -> (ReturnValue:ESlateVisibility Enum)
   )
  ]}
 SetAugmentInfo {nodes [
   n1 = SetAugmentInfo!() -> (then, InAugmentItemDefinition:Fort Player Augment Item Definition Object Reference) {is=title K2Node_FunctionEntry}
   > n2 = Set Augment Item Definition(Augment Item Definition:Fort Player Augment Item Definition Object Reference=@n1) -> (Output_Get:Fort Player Augment Item Definition Object Reference)
   > SetAsEmptySlot!(execute, self, InIsEmptySlot:bool=false) -> (then) {is=title K2Node_CallFunction}
   > UpdateAugment!(
    execute
    self:Fort Player Augment Info Entry Widget Object Reference
    AugmentItemDefinition:Fort Player Augment Item Definition Object Reference=@n2
   ) -> (then)
   > SetAugmentIcon!(
    execute
    self
    Asset:Soft Object Reference=GetHUDEffectImage!(self:Fort Player Augment Item Definition Object Reference=@n2) -> (ReturnValue:Texture 2D Soft Object Reference)
   ) -> (then) {is=title K2Node_CallFunction}
  ]}
 SetEntryFocused {nodes [
   n1 = SetEntryFocused!() -> (then, IsFocused:bool=~) {is=title K2Node_FunctionEntry}
   n2 = Get Anim_Focus() -> (Anim_Focus:Widget Animation Object Reference)
   > PlayAnimation(
    InAnimation=@n2
    StartAtTime=SelectFloat(A=0, B=GetEndTime(self=@n2), bPickA=@n1)
    PlayMode=Select(Option 0:EUMGSequencePlayMode Enum=Reverse, Option 1:EUMGSequencePlayMode Enum, Index=@n1) -> (ReturnValue:EUMGSequencePlayMode Enum)
   )
  ]}
 SetMobileVisuals {nodes [
   SetMobileVisuals() {is=title K2Node_FunctionEntry}
   n1 = AND Boolean(A=IsMobileGame(), B=IsUsingTouch(PlayerController=GetOwningPlayer()))
   > SetVerticalAlignment(
    self=SlotAsOverlaySlot(Widget=Get Overlay_Default_BG() -> (Overlay_Default_BG:Overlay Object Reference))
    InVerticalAlignment=Select(Option 0:EVerticalAlignment Enum=VAlign_Top, Option 1:EVerticalAlignment Enum=VAlign_Fill, Index=@n1) -> (ReturnValue:EVerticalAlignment Enum)
   )
   > SetHorizontalAlignment(
    self=SlotAsOverlaySlot(Widget=Get Overlay_Description() -> (Overlay_Description:Overlay Object Reference))
    InHorizontalAlignment=Select(Option 0:EHorizontalAlignment Enum=HAlign_Left, Option 1:EHorizontalAlignment Enum=HAlign_Fill, Index=@n1) -> (ReturnValue:EHorizontalAlignment Enum)
   )
   > n2 = Branch(Condition=@n1)
   > ClearHeightOverride(self:Size Box Object Reference=Get SizeBox_EntrySize() -> (SizeBox_EntrySize:Size Box Object Reference))
   SetHeightOverride(execute=@n2.else, self=Get SizeBox_EntrySize(), InHeightOverride=115)
  ]}
 UpdateBorrowedPrompt {nodes [
   n1 = UpdateBorrowedPrompt!() -> (then, IsBorrowed:bool=~) {is=title K2Node_FunctionEntry}
   > SetVisibility(
    self=Get Overlay_TeammateAugment() -> (Overlay_TeammateAugment:Overlay Object Reference)
    InVisibility=Select(Option 0:ESlateVisibility Enum=Collapsed, Option 1:ESlateVisibility Enum=SelfHitTestInvisible, Index=@n1) -> (ReturnValue:ESlateVisibility Enum)
   )
  ]}
 UpdateNewPrompt {nodes [
   n1 = UpdateNewPrompt!() -> (then, IsNew:bool=~) {is=title K2Node_FunctionEntry}
   > SetVisibility(
    self=[
     Get Overlay_NewPrompt()
     Get Overlay_UnlockedBang() -> (Overlay_UnlockedBang:Overlay Object Reference)
     Get Icon_Border_Unlocked()
     Get Icon_BackingPanel_UnlockedFX()
     Get Icon_Border_UnlockedGlow()
    ]
    InVisibility=Select(Option 0=Collapsed, Option 1=HitTestInvisible, Index=@n1)
   )
  ]}
]
variables [
 Augment Item Definition {type=FortPlayerAugmentItemDefinition, default=null}
 IsEmptySlot {type=boolean}
]
widgets {
 info {
  inheritedWidgetCount=10
  parentClass=/Script/PlayerAugmentsUI.FortPlayerAugmentInfoEntryWidget
  rootWidgetClass=&Overlay
  widgetCount=65
 }
 widgets [
  Overlay_Main {widgetClassPath=&Overlay, widget=Overlay_Main}
  Overlay_Default_BG {
   widgetClassPath=&Overlay
   bIsVariable=true
   parent=Overlay_Main
   widget=Overlay_Default_BG
   slot {class=OverlaySlot, properties {horizontalAlignment=HAlign_Center, verticalAlignment=VAlign_Center}}
  }
  Backing_Panel {
   widgetClassPath=&Image
   parent=Overlay_Default_BG
   widget=Backing_Panel
   properties {
    visibility=SelfHitTestInvisible
    brush {
     bIsDynamicallyLoaded=false
     drawAs=Box
     imageType=NoImage
     mirroring=NoMirror
     resourceName=None
     resourceObject=/PlayerAugmentsUI/Materials/MI_UI_PlayerAugment_BackingPanel
     tiling=NoTile
     imageSize=<64 64>
     margin=<ltrb 0.5 0.5 0.5 0.5>
     outlineSettings {
      bUseBrushTransparency=false
      roundingType=HalfHeightRadius
      width=0
      color {colorUseRule=UseColor_Specified, specifiedColor=<rgba 0 0 0 0>}
      cornerRadii=<0 0 0 0>
     }
     tintColor {colorUseRule=UseColor_Specified, specifiedColor=<rgba 1 1 1 1>}
     uVRegion {bIsValid=false, max=<0 0>, min=<0 0>}
    }
   }
   slot {class=OverlaySlot, properties {horizontalAlignment=HAlign_Fill, verticalAlignment=VAlign_Fill}}
  }
  Backing_EdgeGlow {
   widgetClassPath=&Image
   parent=Overlay_Default_BG
   widget=Backing_EdgeGlow
   properties {
    visibility=SelfHitTestInvisible
    brush {
     bIsDynamicallyLoaded=false
     drawAs=Box
     imageType=NoImage
     mirroring=NoMirror
     resourceName=None
     resourceObject=/PlayerAugmentsUI/Materials/MI_UI_PlayerAugment_Border_Glow
     tiling=NoTile
     imageSize=<128 128>
     margin=<ltrb 0.5 0.5 0.5 0.5>
     outlineSettings {
      bUseBrushTransparency=false
      roundingType=HalfHeightRadius
      width=0
      color {colorUseRule=UseColor_Specified, specifiedColor=<rgba 0 0 0 0>}
      cornerRadii=<0 0 0 0>
     }
     tintColor {colorUseRule=UseColor_Specified, specifiedColor=<rgba 1 1 1 1>}
     uVRegion {bIsValid=false, max=<0 0>, min=<0 0>}
    }
   }
   slot {class=OverlaySlot, properties {horizontalAlignment=HAlign_Fill, verticalAlignment=VAlign_Fill}}
  }
  WidthConrol {
   widgetClassPath=&Spacer
   parent=Overlay_Default_BG
   widget=WidthConrol
   properties {size=<884 115>}
   slot {class=OverlaySlot, properties {horizontalAlignment=HAlign_Fill, verticalAlignment=VAlign_Center}}
  }
  SizeBox_EntrySize {
   widgetClassPath=&SizeBox
   bIsVariable=true
   parent=Overlay_Main
   widget=SizeBox_EntrySize
   properties {bOverride_HeightOverride=true, bOverride_WidthOverride=true, heightOverride=115, widthOverride=884}
   slot {class=OverlaySlot, properties {horizontalAlignment=HAlign_Center, verticalAlignment=VAlign_Center}}
  }
  HBox_Content {widgetClassPath=&HorizontalBox, parent=SizeBox_EntrySize, widget=HBox_Content, slot {class=SizeBoxSlot}}
  Icon {
   widgetClassPath=&Overlay
   parent=HBox_Content
   widget=Icon
   slot {class=HorizontalBoxSlot, properties {horizontalAlignment=HAlign_Center, verticalAlignment=VAlign_Center, padding=<ltrb 10 0 0 0>}}
  }
  IconSize {
   widgetClassPath=&Spacer
   parent=Icon
   widget=IconSize
   properties {size=<92 92>}
   slot {class=OverlaySlot, properties {horizontalAlignment=HAlign_Center, verticalAlignment=VAlign_Center}}
  }
  Icon_BackingPanel {
   widgetClassPath=&Image
   parent=Icon
   widget=Icon_BackingPanel
   properties {
    visibility=SelfHitTestInvisible
    brush {
     bIsDynamicallyLoaded=false
     drawAs=Box
     imageType=NoImage
     mirroring=NoMirror
     resourceName=None
     resourceObject=/PlayerAugmentsUI/Materials/MI_UI_PlayerAugment_BackingPanel_NoNoise_Icon
     tiling=NoTile
     imageSize=<64 64>
     margin=<ltrb 0.5 0.5 0.5 0.5>
     outlineSettings {
      bUseBrushTransparency=false
      roundingType=HalfHeightRadius
      width=0
      color {colorUseRule=UseColor_Specified, specifiedColor=<rgba 0 0 0 0>}
      cornerRadii=<0 0 0 0>
     }
     tintColor {colorUseRule=UseColor_Specified, specifiedColor=<rgba 1 1 1 1>}
     uVRegion {bIsValid=false, max=<0 0>, min=<0 0>}
    }
   }
   slot {class=OverlaySlot, properties {horizontalAlignment=HAlign_Fill, verticalAlignment=VAlign_Fill}}
  }
  Icon_BackingPanel_UnlockedFX {
   widgetClassPath=&Image
   bIsVariable=true
   parent=Icon
   widget=Icon_BackingPanel_UnlockedFX
   properties {
    visibility=SelfHitTestInvisible
    brush {
     bIsDynamicallyLoaded=false
     drawAs=Box
     imageType=NoImage
     mirroring=NoMirror
     resourceName=None
     resourceObject=/PlayerAugmentsUI/Materials/M_UI_AugmentUnlocked_Pattern
     tiling=NoTile
     imageSize=<64 64>
     margin=<ltrb 0.5 0.5 0.5 0.5>
     outlineSettings {
      bUseBrushTransparency=false
      roundingType=HalfHeightRadius
      width=0
      color {colorUseRule=UseColor_Specified, specifiedColor=<rgba 0 0 0 0>}
      cornerRadii=<0 0 0 0>
     }
     tintColor {colorUseRule=UseColor_Specified, specifiedColor=<rgba 1 1 1 1>}
     uVRegion {bIsValid=false, max=<0 0>, min=<0 0>}
    }
   }
   slot {class=OverlaySlot, properties {horizontalAlignment=HAlign_Fill, verticalAlignment=VAlign_Fill}}
  }
  Icon_Border_UnlockedGlow {
   widgetClassPath=&Image
   bIsVariable=true
   parent=Icon
   widget=Icon_Border_UnlockedGlow
   properties {
    visibility=SelfHitTestInvisible
    brush {
     bIsDynamicallyLoaded=false
     drawAs=Box
     imageType=NoImage
     mirroring=NoMirror
     resourceName=None
     resourceObject=/PlayerAugmentsUI/Materials/MI_UI_PlayerAugment_Unlocked_IconBacking_Glow_Border
     tiling=NoTile
     imageSize=<64 64>
     margin=<ltrb 0.5 0.5 0.5 0.5>
     outlineSettings {
      bUseBrushTransparency=false
      roundingType=HalfHeightRadius
      width=0
      color {colorUseRule=UseColor_Specified, specifiedColor=<rgba 0 0 0 0>}
      cornerRadii=<0 0 0 0>
     }
     tintColor {colorUseRule=UseColor_Specified, specifiedColor=<rgba 1 1 1 1>}
     uVRegion {bIsValid=false, max=<0 0>, min=<0 0>}
    }
   }
   slot {class=OverlaySlot, properties {horizontalAlignment=HAlign_Fill, verticalAlignment=VAlign_Fill, padding=<ltrb -16 -16 -16 -16>}}
  }
  Icon_Border {
   widgetClassPath=&Image
   parent=Icon
   widget=Icon_Border
   properties {
    visibility=SelfHitTestInvisible
    brush {
     bIsDynamicallyLoaded=false
     drawAs=Box
     imageType=NoImage
     mirroring=NoMirror
     resourceName=None
     resourceObject=/PlayerAugmentsUI/Materials/MI_UI_PlayerAugment_Border
     tiling=NoTile
     imageSize=<64 64>
     margin=<ltrb 0.5 0.5 0.5 0.5>
     outlineSettings {
      bUseBrushTransparency=false
      roundingType=HalfHeightRadius
      width=0
      color {colorUseRule=UseColor_Specified, specifiedColor=<rgba 0 0 0 0>}
      cornerRadii=<0 0 0 0>
     }
     tintColor {colorUseRule=UseColor_Specified, specifiedColor=<rgba 1 1 1 1>}
     uVRegion {bIsValid=false, max=<0 0>, min=<0 0>}
    }
    renderTransformPivot=<1 1>
   }
   slot {class=OverlaySlot, properties {horizontalAlignment=HAlign_Fill, verticalAlignment=VAlign_Fill}}
  }
  Icon_Border_Unlocked {
   widgetClassPath=&Image
   bIsVariable=true
   parent=Icon
   widget=Icon_Border_Unlocked
   properties {
    visibility=SelfHitTestInvisible
    brush {
     bIsDynamicallyLoaded=false
     drawAs=Box
     imageType=NoImage
     mirroring=NoMirror
     resourceName=None
     resourceObject=/PlayerAugmentsUI/Materials/MI_UI_PlayerAugment_Border_Unlcoked
     tiling=NoTile
     imageSize=<64 64>
     margin=<ltrb 0.5 0.5 0.5 0.5>
     outlineSettings {
      bUseBrushTransparency=false
      roundingType=HalfHeightRadius
      width=0
      color {colorUseRule=UseColor_Specified, specifiedColor=<rgba 0 0 0 0>}
      cornerRadii=<0 0 0 0>
     }
     tintColor {colorUseRule=UseColor_Specified, specifiedColor=<rgba 1 1 1 1>}
     uVRegion {bIsValid=false, max=<0 0>, min=<0 0>}
    }
    renderTransformPivot=<1 1>
   }
   slot {class=OverlaySlot, properties {horizontalAlignment=HAlign_Fill, verticalAlignment=VAlign_Fill}}
  }
  Image_AugmentIcon {
   widgetClassPath=&Image
   bIsVariable=true
   parent=Icon
   widget=Image_AugmentIcon
   properties {
    visibility=SelfHitTestInvisible
    brush {
     bIsDynamicallyLoaded=false
     drawAs=Image
     imageType=NoImage
     mirroring=NoMirror
     resourceName=None
     resourceObject=/PlayerAugmentsUI/Materials/MI_UI_PlayerAugment_HUD_Icon
     tiling=NoTile
     imageSize=<92 92>
     margin=<ltrb 0 0 0 0>
     outlineSettings {
      bUseBrushTransparency=false
      roundingType=HalfHeightRadius
      width=0
      color {colorUseRule=UseColor_Specified, specifiedColor=<rgba 0 0 0 0>}
      cornerRadii=<0 0 0 0>
     }
     tintColor {colorUseRule=UseColor_Specified, specifiedColor=<rgba 1 1 1 1>}
     uVRegion {bIsValid=false, max=<0 0>, min=<0 0>}
    }
   }
   slot {class=OverlaySlot, properties {horizontalAlignment=HAlign_Center, verticalAlignment=VAlign_Center}}
  }
  Overlay_NewPrompt {
   widgetClassPath=&Overlay
   bIsVariable=true
   parent=Icon
   widget=Overlay_NewPrompt
   properties {renderTransform {angle=0, scale=<1 1>, shear=<0 0>, translation=<0 12>}}
   slot {class=OverlaySlot, properties {horizontalAlignment=HAlign_Center, verticalAlignment=VAlign_Bottom}}
  }
  Image_1 {
   widgetClassPath=&Image
   parent=Overlay_NewPrompt
   widget=Image_1
   properties {
    visibility=SelfHitTestInvisible
    brush {
     bIsDynamicallyLoaded=false
     drawAs=Box
     imageType=NoImage
     mirroring=NoMirror
     resourceName=None
     resourceObject=/PlayerAugmentsUI/Materials/MI_UI_PlayerAugment_Unlocked_IconBacking_Glow
     tiling=NoTile
     imageSize=<48 48>
     margin=<ltrb 0.5 0.5 0.5 0.5>
     outlineSettings {
      bUseBrushTransparency=false
      roundingType=HalfHeightRadius
      width=0
      color {colorUseRule=UseColor_Specified, specifiedColor=<rgba 0 0 0 0>}
      cornerRadii=<0 0 0 0>
     }
     tintColor {colorUseRule=UseColor_Specified, specifiedColor=<rgba 1 1 1 1>}
     uVRegion {bIsValid=false, max=<0 0>, min=<0 0>}
    }
   }
   slot {class=OverlaySlot, properties {horizontalAlignment=HAlign_Fill, verticalAlignment=VAlign_Fill, padding=<ltrb -16 -16 -16 -16>}}
  }
  Image_410 {
   widgetClassPath=&Image
   parent=Overlay_NewPrompt
   widget=Image_410
   properties {
    visibility=SelfHitTestInvisible
    brush {
     bIsDynamicallyLoaded=false
     drawAs=Box
     imageType=NoImage
     mirroring=NoMirror
     resourceName=None
     resourceObject=/PlayerAugmentsUI/Materials/MI_UI_PlayerAugment_NotificationBacking
     tiling=NoTile
     imageSize=<16 16>
     margin=<ltrb 0.5 0.5 0.5 0.5>
     outlineSettings {
      bUseBrushTransparency=false
      roundingType=HalfHeightRadius
      width=0
      color {colorUseRule=UseColor_Specified, specifiedColor=<rgba 0 0 0 0>}
      cornerRadii=<0 0 0 0>
     }
     tintColor {colorUseRule=UseColor_Specified, specifiedColor=<rgba 1 1 1 1>}
     uVRegion {bIsValid=false, max=<0 0>, min=<0 0>}
    }
   }
   slot {class=OverlaySlot, properties {horizontalAlignment=HAlign_Fill, verticalAlignment=VAlign_Fill}}
  }
  Image {
   widgetClassPath=&Image
   parent=Overlay_NewPrompt
   widget=Image
   properties {
    visibility=SelfHitTestInvisible
    brush {
     bIsDynamicallyLoaded=false
     drawAs=Box
     imageType=NoImage
     mirroring=NoMirror
     resourceName=None
     resourceObject=/PlayerAugmentsUI/Materials/MI_UI_PlayerAugment_Notification_Border
     tiling=NoTile
     imageSize=<16 16>
     margin=<ltrb 0.5 0.5 0.5 0.5>
     outlineSettings {
      bUseBrushTransparency=false
      roundingType=HalfHeightRadius
      width=0
      color {colorUseRule=UseColor_Specified, specifiedColor=<rgba 0 0 0 0>}
      cornerRadii=<0 0 0 0>
     }
     tintColor {colorUseRule=UseColor_Specified, specifiedColor=<rgba 1 1 1 1>}
     uVRegion {bIsValid=false, max=<0 0>, min=<0 0>}
    }
   }
   slot {class=OverlaySlot, properties {horizontalAlignment=HAlign_Fill, verticalAlignment=VAlign_Fill}}
  }
  SizeBox_0 {
   widgetClassPath=&SizeBox
   parent=Overlay_NewPrompt
   widget=SizeBox_0
   properties {bOverride_MaxDesiredWidth=true, maxDesiredWidth=56, widthOverride=112}
   slot {class=OverlaySlot, properties {horizontalAlignment=HAlign_Center, padding=<ltrb 8 6 8 6>}}
  }
  ScaleBox_0 {
   widgetClassPath=&ScaleBox
   parent=SizeBox_0
   widget=ScaleBox_0
   properties {stretch=ScaleToFitX, stretchDirection=DownOnly}
   slot {class=SizeBoxSlot}
  }
  Text_NewAugment {
   widgetClassPath=&CommonTextBlock
   parent=ScaleBox_0
   widget=Text_NewAugment
   properties {
    mobileFontSizeMultiplier=1.4
    text="Unlocked!"
    textTransformPolicy=ToUpper
    font {
     bForceMonospaced=false
     bMaterialIsStencil=false
     fontMaterial=/PlayerAugmentsUI/Materials/MI_UI_PlayerAugment_Notification_Small_Font
     fontObject=&BurbankSmall
     letterSpacing=0
     monospacedWidth=1
     size=10
     skewAmount=0
     typefaceFontName=Black
     outlineSettings {
      bApplyOutlineToDropShadows=false
      bMiteredCorners=false
      bSeparateFillAlpha=true
      outlineMaterial=null
      outlineSize=0
      outlineColor=<rgba 0 0 0 0>
     }
    }
    shadowOffset=<0 0>
   }
   slot {class=ScaleBoxSlot}
  }
  VisSwitcher_Description {
   widgetClassPath=&CommonVisibilitySwitcher
   bIsVariable=true
   parent=HBox_Content
   widget=VisSwitcher_Description
   slot {class=HorizontalBoxSlot, properties {verticalAlignment=VAlign_Center, padding=<ltrb 24 0 0 0>}}
  }
  VBox_AugmentText {
   widgetClassPath=&VerticalBox
   bIsVariable=true
   parent=VisSwitcher_Description
   widget=VBox_AugmentText
   slot {class=CommonVisibilitySwitcherSlot}
  }
  HorizontalBox_0 {
   widgetClassPath=&HorizontalBox
   parent=VBox_AugmentText
   widget=HorizontalBox_0
   slot {class=VerticalBoxSlot, properties {verticalAlignment=VAlign_Top}}
  }
  Overlay_0 {widgetClassPath=&Overlay, parent=HorizontalBox_0, widget=Overlay_0, slot {class=HorizontalBoxSlot}}
  Overlay_Header {widgetClassPath=&Overlay, parent=Overlay_0, widget=Overlay_Header, slot {class=OverlaySlot}}
  Text_Glow {
   widgetClassPath=&Image
   parent=Overlay_Header
   widget=Text_Glow
   properties {brush {
     bIsDynamicallyLoaded=false
     drawAs=Box
     imageType=NoImage
     mirroring=NoMirror
     resourceName=None
     resourceObject=/PlayerAugmentsUI/Materials/MI_UI_PlayerAugment_Font_Glow
     tiling=NoTile
     imageSize=<64 64>
     margin=<ltrb 0.5 0.5 0.5 0.5>
     outlineSettings {
      bUseBrushTransparency=false
      roundingType=HalfHeightRadius
      width=0
      color {colorUseRule=UseColor_Specified, specifiedColor=<rgba 0 0 0 0>}
      cornerRadii=<0 0 0 0>
     }
     tintColor {colorUseRule=UseColor_Specified, specifiedColor=<rgba 1 1 1 1>}
     uVRegion {bIsValid=false, max=<0 0>, min=<0 0>}
    }}
   slot {class=OverlaySlot, properties {horizontalAlignment=HAlign_Fill, verticalAlignment=VAlign_Fill, padding=<ltrb -28 -16 -28 -16>}}
  }
  Text_AugmentName {
   widgetClassPath=&CommonTextBlock
   bInherited=true
   bIsVariable=true
   parent=Overlay_Header
   widget=Text_AugmentName
   properties {
    lineHeightPercentage=0.9
    mobileFontSizeMultiplier=1.2
    text="[PH] Augment Name"
    textTransformPolicy=ToUpper
    font {
     bForceMonospaced=false
     bMaterialIsStencil=false
     fontMaterial=null
     fontObject=&Burbank
     letterSpacing=50
     monospacedWidth=1
     size=15
     skewAmount=0
     typefaceFontName=BurbankRegular-Black
     outlineSettings {
      bApplyOutlineToDropShadows=false
      bMiteredCorners=false
      bSeparateFillAlpha=true
      outlineMaterial=&M-FontShadow
      outlineSize=3
      outlineColor=<rgba 0 0.1667 1 1>
     }
    }
    margin=<ltrb 0 -2 0 0>
    shadowColorAndOpacity=<rgba 0 0.8333 1 0.125>
    shadowOffset=<6 6>
   }
   slot {class=OverlaySlot}
  }
  Overlay_UnlockedBang {
   widgetClassPath=&Overlay
   bIsVariable=true
   parent=Overlay_0
   widget=Overlay_UnlockedBang
   properties {renderTransform {angle=0, scale=<1 1>, shear=<0 0>, translation=<22 -10>}}
   slot {class=OverlaySlot, properties {horizontalAlignment=HAlign_Right}}
  }
  Spacer_131 {
   widgetClassPath=&Spacer
   parent=Overlay_UnlockedBang
   widget=Spacer_131
   properties {size=<24 24>}
   slot {class=OverlaySlot, properties {horizontalAlignment=HAlign_Fill, verticalAlignment=VAlign_Fill}}
  }
  Overlay {
   widgetClassPath=&Overlay
   parent=Overlay_UnlockedBang
   widget=Overlay
   slot {class=OverlaySlot, properties {horizontalAlignment=HAlign_Fill, verticalAlignment=VAlign_Fill}}
  }
  Image_2 {
   widgetClassPath=&Image
   parent=Overlay
   widget=Image_2
   properties {
    visibility=SelfHitTestInvisible
    brush {
     bIsDynamicallyLoaded=false
     drawAs=Box
     imageType=NoImage
     mirroring=NoMirror
     resourceName=None
     resourceObject=&MI_UI_PlayerAugment_Unlocked_IconBacking_Glow
     tiling=NoTile
     imageSize=<48 48>
     margin=<ltrb 0.5 0.5 0.5 0.5>
     outlineSettings {
      bUseBrushTransparency=false
      roundingType=HalfHeightRadius
      width=0
      color {colorUseRule=UseColor_Specified, specifiedColor=<rgba 0 0 0 0>}
      cornerRadii=<0 0 0 0>
     }
     tintColor {colorUseRule=UseColor_Specified, specifiedColor=<rgba 1 1 1 1>}
     uVRegion {bIsValid=false, max=<0 0>, min=<0 0>}
    }
   }
   slot {class=OverlaySlot, properties {horizontalAlignment=HAlign_Fill, verticalAlignment=VAlign_Fill, padding=<ltrb -16 -16 -16 -16>}}
  }
  Image_3 {
   widgetClassPath=&Image
   parent=Overlay
   widget=Image_3
   properties {
    visibility=SelfHitTestInvisible
    brush {
     bIsDynamicallyLoaded=false
     drawAs=Box
     imageType=NoImage
     mirroring=NoMirror
     resourceName=None
     resourceObject=&MI_UI_PlayerAugment_NotificationBacking
     tiling=NoTile
     imageSize=<16 16>
     margin=<ltrb 0.5 0.5 0.5 0.5>
     outlineSettings {
      bUseBrushTransparency=false
      roundingType=HalfHeightRadius
      width=0
      color {colorUseRule=UseColor_Specified, specifiedColor=<rgba 0 0 0 0>}
      cornerRadii=<0 0 0 0>
     }
     tintColor {colorUseRule=UseColor_Specified, specifiedColor=<rgba 1 1 1 1>}
     uVRegion {bIsValid=false, max=<0 0>, min=<0 0>}
    }
   }
   slot {class=OverlaySlot, properties {horizontalAlignment=HAlign_Fill, verticalAlignment=VAlign_Fill}}
  }
  Image_4 {
   widgetClassPath=&Image
   parent=Overlay
   widget=Image_4
   properties {
    visibility=SelfHitTestInvisible
    brush {
     bIsDynamicallyLoaded=false
     drawAs=Box
     imageType=NoImage
     mirroring=NoMirror
     resourceName=None
     resourceObject=&MI_UI_PlayerAugment_Notification_Border
     tiling=NoTile
     imageSize=<16 16>
     margin=<ltrb 0.5 0.5 0.5 0.5>
     outlineSettings {
      bUseBrushTransparency=false
      roundingType=HalfHeightRadius
      width=0
      color {colorUseRule=UseColor_Specified, specifiedColor=<rgba 0 0 0 0>}
      cornerRadii=<0 0 0 0>
     }
     tintColor {colorUseRule=UseColor_Specified, specifiedColor=<rgba 1 1 1 1>}
     uVRegion {bIsValid=false, max=<0 0>, min=<0 0>}
    }
   }
   slot {class=OverlaySlot, properties {horizontalAlignment=HAlign_Fill, verticalAlignment=VAlign_Fill}}
  }
  Text_NewAugment_1 {
   widgetClassPath=&CommonTextBlock
   parent=Overlay_UnlockedBang
   widget=Text_NewAugment_1
   properties {
    text="!"
    textTransformPolicy=ToUpper
    font {
     bForceMonospaced=false
     bMaterialIsStencil=false
     fontMaterial=&MI_UI_PlayerAugment_Notification_Small_Font
     fontObject=&BurbankSmall
     letterSpacing=0
     monospacedWidth=1
     size=13
     skewAmount=0
     typefaceFontName=Black
     outlineSettings {
      bApplyOutlineToDropShadows=false
      bMiteredCorners=false
      bSeparateFillAlpha=true
      outlineMaterial=null
      outlineSize=0
      outlineColor=<rgba 0 0 0 0>
     }
    }
    shadowOffset=<0 0>
   }
   slot {class=OverlaySlot, properties {horizontalAlignment=HAlign_Center, verticalAlignment=VAlign_Center}}
  }
  Overlay_TeammateAugment {
   widgetClassPath=&Overlay
   bIsVariable=true
   parent=HorizontalBox_0
   widget=Overlay_TeammateAugment
   properties {visibility=Collapsed}
   slot {class=HorizontalBoxSlot, properties {horizontalAlignment=HAlign_Right, verticalAlignment=VAlign_Center, size {sizeRule=Fill, value=1}}}
  }
  HorizontalBox {
   widgetClassPath=&HorizontalBox
   parent=Overlay_TeammateAugment
   widget=HorizontalBox
   properties {flowDirectionPreference=Culture}
   slot {class=OverlaySlot, properties {horizontalAlignment=HAlign_Center, verticalAlignment=VAlign_Center}}
  }
  Overlay_5 {
   widgetClassPath=&Overlay
   parent=HorizontalBox
   widget=Overlay_5
   slot {class=HorizontalBoxSlot, properties {horizontalAlignment=HAlign_Left, verticalAlignment=VAlign_Center}}
  }
  Text_Glow_1 {
   widgetClassPath=&Image
   parent=Overlay_5
   widget=Text_Glow_1
   properties {
    visibility=SelfHitTestInvisible
    brush {
     bIsDynamicallyLoaded=false
     drawAs=Box
     imageType=NoImage
     mirroring=NoMirror
     resourceName=None
     resourceObject=&MI_UI_PlayerAugment_Font_Glow
     tiling=NoTile
     imageSize=<40 40>
     margin=<ltrb 0.5 0.5 0.5 0.5>
     outlineSettings {
      bUseBrushTransparency=false
      roundingType=HalfHeightRadius
      width=0
      color {colorUseRule=UseColor_Specified, specifiedColor=<rgba 0 0 0 0>}
      cornerRadii=<0 0 0 0>
     }
     tintColor {colorUseRule=UseColor_Specified, specifiedColor=<rgba 1 1 1 1>}
     uVRegion {bIsValid=false, max=<0 0>, min=<0 0>}
    }
   }
   slot {class=OverlaySlot, properties {horizontalAlignment=HAlign_Fill, verticalAlignment=VAlign_Fill, padding=<ltrb -8 -6 -8 -6>}}
  }
  Text_Borrow {
   widgetClassPath=&CommonTextBlock
   parent=Overlay_5
   widget=Text_Borrow
   properties {
    mobileFontSizeMultiplier=1.3
    text=Borrowed from
    textTransformPolicy=ToUpper
    font {
     bForceMonospaced=false
     bMaterialIsStencil=false
     fontMaterial=/PlayerAugmentsUI/Materials/MI_UI_PlayerAugment_SocialCallout_Font
     fontObject=&BurbankSmall
     letterSpacing=0
     monospacedWidth=1
     size=10
     skewAmount=0
     typefaceFontName=Bold
     outlineSettings {
      bApplyOutlineToDropShadows=false
      bMiteredCorners=false
      bSeparateFillAlpha=true
      outlineMaterial=&M-FontShadow
      outlineSize=2
      outlineColor=<rgba 0 0.1667 1 1>
     }
    }
    shadowColorAndOpacity=<rgba 0 0.8333 1 0.125>
    shadowOffset=<6 4>
   }
   slot {class=OverlaySlot, properties {verticalAlignment=VAlign_Bottom}}
  }
  Overlay_LenderAvatarIcon {
   widgetClassPath=&Overlay
   bInherited=true
   parent=HorizontalBox
   widget=Overlay_LenderAvatarIcon
   slot {class=HorizontalBoxSlot, properties {verticalAlignment=VAlign_Center, padding=<ltrb 0 0 5 0>}}
  }
  Image_TeammateIcon_Color_Glow {
   widgetClassPath=&Image
   bIsVariable=true
   parent=Overlay_LenderAvatarIcon
   widget=Image_TeammateIcon_Color_Glow
   properties {
    visibility=SelfHitTestInvisible
    brush {
     bIsDynamicallyLoaded=false
     drawAs=Image
     imageType=NoImage
     mirroring=NoMirror
     resourceName=None
     resourceObject=/PlayerAugmentsUI/Materials/MI_UI_Augment_SocialColorBacking_Glow
     tiling=NoTile
     imageSize=<24 24>
     margin=<ltrb 0 0 0 0>
     outlineSettings {
      bUseBrushTransparency=false
      roundingType=HalfHeightRadius
      width=0
      color {colorUseRule=UseColor_Specified, specifiedColor=<rgba 0 0 0 0>}
      cornerRadii=<0 0 0 0>
     }
     tintColor {colorUseRule=UseColor_Specified, specifiedColor=<rgba 1 1 1 1>}
     uVRegion {bIsValid=false, max=<0 0>, min=<0 0>}
    }
    renderTransform {angle=0, scale=<2 2>, shear=<0 0>, translation=<0 -2>}
   }
   slot {class=OverlaySlot, properties {horizontalAlignment=HAlign_Fill, verticalAlignment=VAlign_Fill}}
  }
  Image_TeammateIcon_Color {
   widgetClassPath=&Image
   bIsVariable=true
   parent=Overlay_LenderAvatarIcon
   widget=Image_TeammateIcon_Color
   properties {
    visibility=SelfHitTestInvisible
    brush {
     bIsDynamicallyLoaded=false
     drawAs=Image
     imageType=NoImage
     mirroring=NoMirror
     resourceName=None
     resourceObject=/PlayerAugmentsUI/Materials/MI_UI_Augment_SocialColorBacking
     tiling=NoTile
     imageSize=<24 24>
     margin=<ltrb 0 0 0 0>
     outlineSettings {
      bUseBrushTransparency=false
      roundingType=HalfHeightRadius
      width=0
      color {colorUseRule=UseColor_Specified, specifiedColor=<rgba 0 0 0 0>}
      cornerRadii=<0 0 0 0>
     }
     tintColor {colorUseRule=UseColor_Specified, specifiedColor=<rgba 1 1 1 1>}
     uVRegion {bIsValid=false, max=<0 0>, min=<0 0>}
    }
    renderTransform {angle=0, scale=<1.125 1.125>, shear=<0 0>, translation=<0 -2>}
   }
   slot {class=OverlaySlot, properties {horizontalAlignment=HAlign_Center, verticalAlignment=VAlign_Center}}
  }
  Icon_LenderAvatarIcon {
   widgetClassPath=/Script/FortniteUI.FortSocialAvatarIcon
   bInherited=true
   parent=Overlay_LenderAvatarIcon
   widget=Icon_LenderAvatarIcon
   properties {
    brush {
     bIsDynamicallyLoaded=false
     drawAs=Image
     imageType=NoImage
     mirroring=NoMirror
     resourceName=None
     resourceObject=/Game/UI/SocialPanel/Art/Materials/UI_UserPortrait_Face_NoBackground
     tiling=NoTile
     imageSize=<32 32>
     margin=<ltrb 0 0 0 0>
     outlineSettings {
      bUseBrushTransparency=false
      roundingType=HalfHeightRadius
      width=0
      color {colorUseRule=UseColor_Specified, specifiedColor=<rgba 0 0 0 0>}
      cornerRadii=<0 0 0 0>
     }
     tintColor {colorUseRule=UseColor_Specified, specifiedColor=<rgba 1 1 1 1>}
     uVRegion {bIsValid=false, max=<0 0>, min=<0 0>}
    }
    renderTransform {angle=0, scale=<1 1>, shear=<0 0>, translation=<0 -2>}
   }
   slot {class=OverlaySlot, properties {horizontalAlignment=HAlign_Center, verticalAlignment=VAlign_Center}}
  }
  Overlay_6 {
   widgetClassPath=&Overlay
   parent=HorizontalBox
   widget=Overlay_6
   slot {class=HorizontalBoxSlot, properties {horizontalAlignment=HAlign_Left, verticalAlignment=VAlign_Center}}
  }
  Text_Glow_2 {
   widgetClassPath=&Image
   parent=Overlay_6
   widget=Text_Glow_2
   properties {
    visibility=SelfHitTestInvisible
    brush {
     bIsDynamicallyLoaded=false
     drawAs=Box
     imageType=NoImage
     mirroring=NoMirror
     resourceName=None
     resourceObject=&MI_UI_PlayerAugment_Font_Glow
     tiling=NoTile
     imageSize=<40 40>
     margin=<ltrb 0.5 0.5 0.5 0.5>
     outlineSettings {
      bUseBrushTransparency=false
      roundingType=HalfHeightRadius
      width=0
      color {colorUseRule=UseColor_Specified, specifiedColor=<rgba 0 0 0 0>}
      cornerRadii=<0 0 0 0>
     }
     tintColor {colorUseRule=UseColor_Specified, specifiedColor=<rgba 1 1 1 1>}
     uVRegion {bIsValid=false, max=<0 0>, min=<0 0>}
    }
   }
   slot {class=OverlaySlot, properties {horizontalAlignment=HAlign_Fill, verticalAlignment=VAlign_Fill, padding=<ltrb -8 -6 -8 -6>}}
  }
  SizeBox_1 {
   widgetClassPath=&SizeBox
   parent=Overlay_6
   widget=SizeBox_1
   properties {bOverride_WidthOverride=true, widthOverride=162}
   slot {class=OverlaySlot, properties {verticalAlignment=VAlign_Center}}
  }
  ScaleBox_1 {
   widgetClassPath=&ScaleBox
   parent=SizeBox_1
   widget=ScaleBox_1
   properties {stretchDirection=DownOnly}
   slot {class=SizeBoxSlot, properties {verticalAlignment=VAlign_Center}}
  }
  Text_BorrowingText {
   widgetClassPath=&CommonTextBlock
   bInherited=true
   bIsVariable=true
   parent=ScaleBox_1
   widget=Text_BorrowingText
   properties {
    mobileFontSizeMultiplier=1.3
    text=MMWWMMWWMMWWMMWW
    textTransformPolicy=ToUpper
    font {
     bForceMonospaced=false
     bMaterialIsStencil=false
     fontMaterial=&MI_UI_PlayerAugment_SocialCallout_Font
     fontObject=&BurbankSmall
     letterSpacing=0
     monospacedWidth=1
     size=10
     skewAmount=0
     typefaceFontName=Bold
     outlineSettings {
      bApplyOutlineToDropShadows=false
      bMiteredCorners=false
      bSeparateFillAlpha=true
      outlineMaterial=&M-FontShadow
      outlineSize=3
      outlineColor=<rgba 0 0.1651 1 1>
     }
    }
    shadowColorAndOpacity=<rgba 0 0.8333 1 0.125>
    shadowOffset=<6 4>
   }
   slot {class=ScaleBoxSlot, properties {horizontalAlignment=HAlign_Left}}
  }
  Overlay_DescriptionText {
   widgetClassPath=&Overlay
   bIsVariable=true
   parent=VBox_AugmentText
   widget=Overlay_DescriptionText
   slot {class=VerticalBoxSlot, properties {horizontalAlignment=HAlign_Left, verticalAlignment=VAlign_Top}}
  }
  DescriptionTextSize {
   widgetClassPath=&Spacer
   parent=Overlay_DescriptionText
   widget=DescriptionTextSize
   properties {size=<746 1>}
   slot {class=OverlaySlot, properties {horizontalAlignment=HAlign_Fill, verticalAlignment=VAlign_Fill}}
  }
  Overlay_Description {widgetClassPath=&Overlay, bIsVariable=true, parent=Overlay_DescriptionText, widget=Overlay_Description, slot {class=OverlaySlot}}
  Image_SolidBG_AugmentDescription_Reflection {
   widgetClassPath=&Image
   parent=Overlay_Description
   widget=Image_SolidBG_AugmentDescription_Reflection
   properties {
    visibility=SelfHitTestInvisible
    brush {
     bIsDynamicallyLoaded=false
     drawAs=Box
     imageType=NoImage
     mirroring=NoMirror
     resourceName=None
     resourceObject=/PlayerAugmentsUI/Materials/MI_UI_PlayerAugment_BackingPanel_NoNoise_Reflection_TextPanel
     tiling=NoTile
     imageSize=<32 32>
     margin=<ltrb 0.5 0.5 0.5 0.5>
     outlineSettings {
      bUseBrushTransparency=false
      roundingType=HalfHeightRadius
      width=0
      color {colorUseRule=UseColor_Specified, specifiedColor=<rgba 0 0 0 0>}
      cornerRadii=<0 0 0 0>
     }
     tintColor {colorUseRule=UseColor_Specified, specifiedColor=<rgba 1 1 1 1>}
     uVRegion {bIsValid=false, max=<0 0>, min=<0 0>}
    }
    renderTransform {angle=0, scale=<1 1>, shear=<0 0>, translation=<8 12>}
   }
   slot {class=OverlaySlot, properties {horizontalAlignment=HAlign_Fill, verticalAlignment=VAlign_Fill}}
  }
  Image_SolidBG_AugmentDescription {
   widgetClassPath=&Image
   parent=Overlay_Description
   widget=Image_SolidBG_AugmentDescription
   properties {
    visibility=SelfHitTestInvisible
    brush {
     bIsDynamicallyLoaded=false
     drawAs=Box
     imageType=NoImage
     mirroring=NoMirror
     resourceName=None
     resourceObject=/PlayerAugmentsUI/Materials/MI_UI_PlayerAugment_BackingPanel_NoNoise
     tiling=NoTile
     imageSize=<32 32>
     margin=<ltrb 0.5 0.5 0.5 0.5>
     outlineSettings {
      bUseBrushTransparency=false
      roundingType=HalfHeightRadius
      width=0
      color {colorUseRule=UseColor_Specified, specifiedColor=<rgba 0 0 0 0>}
      cornerRadii=<0 0 0 0>
     }
     tintColor {colorUseRule=UseColor_Specified, specifiedColor=<rgba 1 1 1 1>}
     uVRegion {bIsValid=false, max=<0 0>, min=<0 0>}
    }
   }
   slot {class=OverlaySlot, properties {horizontalAlignment=HAlign_Fill, verticalAlignment=VAlign_Fill}}
  }
  TextPanelBrackets {
   widgetClassPath=&Image
   parent=Overlay_Description
   widget=TextPanelBrackets
   properties {
    visibility=SelfHitTestInvisible
    brush {
     bIsDynamicallyLoaded=false
     drawAs=Box
     imageType=NoImage
     mirroring=NoMirror
     resourceName=None
     resourceObject=/PlayerAugmentsUI/Materials/MI_UI_PlayerAugment_Brackets
     tiling=NoTile
     imageSize=<32 32>
     margin=<ltrb 0.5 0.5 0.5 0.5>
     outlineSettings {
      bUseBrushTransparency=false
      roundingType=HalfHeightRadius
      width=0
      color {colorUseRule=UseColor_Specified, specifiedColor=<rgba 0 0 0 0>}
      cornerRadii=<0 0 0 0>
     }
     tintColor {colorUseRule=UseColor_Specified, specifiedColor=<rgba 1 1 1 1>}
     uVRegion {bIsValid=false, max=<0 0>, min=<0 0>}
    }
   }
   slot {class=OverlaySlot, properties {horizontalAlignment=HAlign_Fill, verticalAlignment=VAlign_Fill}}
  }
  VerticalBox_0 {
   widgetClassPath=&VerticalBox
   parent=Overlay_Description
   widget=VerticalBox_0
   slot {class=OverlaySlot, properties {horizontalAlignment=HAlign_Fill, padding=<ltrb 12 8 12 8>}}
  }
  Text_AugmentCategory {
   widgetClassPath=&CommonTextBlock
   bInherited=true
   bIsVariable=true
   parent=VerticalBox_0
   widget=Text_AugmentCategory
   properties {
    lineHeightPercentage=0.9
    mobileFontSizeMultiplier=1.45
    text="[PH] Augment Category"
    textTransformPolicy=ToUpper
    wrapTextAt=754
    font {
     bForceMonospaced=false
     bMaterialIsStencil=false
     fontMaterial=null
     fontObject=&BurbankSmall
     letterSpacing=0
     monospacedWidth=1
     size=11
     skewAmount=0.2
     typefaceFontName=Bold
     outlineSettings {
      bApplyOutlineToDropShadows=false
      bMiteredCorners=false
      bSeparateFillAlpha=true
      outlineMaterial=&M-FontShadow
      outlineSize=2
      outlineColor=<rgba 0 0.1667 1 1>
     }
    }
    shadowColorAndOpacity=<rgba 0 0.8333 1 0.125>
    shadowOffset=<6 4>
   }
   slot {class=VerticalBoxSlot, properties {verticalAlignment=VAlign_Top}}
  }
  Text_AugmentDescription {
   widgetClassPath=&CommonTextBlock
   bInherited=true
   bIsVariable=true
   parent=VerticalBox_0
   widget=Text_AugmentDescription
   properties {
    lineHeightPercentage=0.9
    mobileFontSizeMultiplier=1.3
    text="[PH] Augment Description. Example text to see how long it will go before wrapping and test multi lines. Where are the multiple lines"
    wrapTextAt=754
    font {
     bForceMonospaced=false
     bMaterialIsStencil=false
     fontMaterial=/PlayerAugmentsUI/Materials/MI_UI_PlayerAugment_Body_Small_Font
     fontObject=&BurbankSmall
     letterSpacing=0
     monospacedWidth=1
     size=11
     skewAmount=0
     typefaceFontName=Bold
     outlineSettings {
      bApplyOutlineToDropShadows=false
      bMiteredCorners=false
      bSeparateFillAlpha=true
      outlineMaterial=&M-FontShadow
      outlineSize=2
      outlineColor=<rgba 0 0.1667 1 1>
     }
    }
    shadowColorAndOpacity=<rgba 0 0.8333 1 0.125>
    shadowOffset=<6 4>
   }
   slot {class=VerticalBoxSlot, properties {horizontalAlignment=HAlign_Left}}
  }
  Overlay_EmptyPrompt {
   widgetClassPath=&Overlay
   bIsVariable=true
   parent=VisSwitcher_Description
   widget=Overlay_EmptyPrompt
   slot {class=CommonVisibilitySwitcherSlot, properties {verticalAlignment=VAlign_Center}}
  }
  Text_Borrow_1 {
   widgetClassPath=&CommonTextBlock
   parent=Overlay_EmptyPrompt
   widget=Text_Borrow_1
   properties {
    mobileFontSizeMultiplier=1.3
    text=Empty
    textTransformPolicy=ToUpper
    colorAndOpacity {colorUseRule=UseColor_Specified, specifiedColor=<rgba 1 1 1 0.75>}
    font {
     bForceMonospaced=false
     bMaterialIsStencil=false
     fontMaterial=null
     fontObject=&BurbankSmall
     letterSpacing=0
     monospacedWidth=1
     size=13
     skewAmount=0
     typefaceFontName=Bold
     outlineSettings {
      bApplyOutlineToDropShadows=false
      bMiteredCorners=false
      bSeparateFillAlpha=true
      outlineMaterial=&M-FontShadow
      outlineSize=2
      outlineColor=<rgba 0 0.1667 1 1>
     }
    }
    shadowColorAndOpacity=<rgba 0 0.8333 1 0.125>
    shadowOffset=<6 4>
   }
   slot {class=OverlaySlot}
  }
  Text_AugmentRarity {widgetClassPath=&CommonTextBlock, bInherited=true, widget=null}
  ItemCategoryIndicator {widgetClassPath=/Script/FortniteUI.FortItemCategoryIndicator, bInherited=true, widget=null}
  Image_CategoryAsImage {widgetClassPath=&FortLazyImage, bInherited=true, widget=null}
  Image_LenderSquadColor {widgetClassPath=&FortLazyImage, bInherited=true, widget=null}
 ]
}
